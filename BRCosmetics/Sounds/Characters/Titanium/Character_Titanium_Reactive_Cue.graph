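# Sound cue for the Titanium character's reactive sound in BRCosmetics.
# A root Random node picks between three equally weighted Wave Players, one for each Reactive wave variation.
# Playback is routed through the Fort_Gadgets sound class and uses the shared basic Foley attenuation to fall off with distance.
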SoundCue /BRCosmetics/Sounds/Characters/Titanium/Character_Titanium_Reactive_Cue
root=Random_0
nodes [
 {id=Random_0, children [WavePlayer_0,WavePlayer_1,WavePlayer_2], properties {weights [1,1,1]}}
 {id=WavePlayer_0, properties {soundWaveAssetPtr=&/Character_Titanium_Reactive_03}}
 {id=WavePlayer_1, properties {soundWaveAssetPtr=&/Character_Titanium_Reactive_01}}
 {id=WavePlayer_2, properties {soundWaveAssetPtr=&/Character_Titanium_Reactive_02}}
]
properties {
 attenuationSettings=&Foley_Attenuation_Basic_Att
 maxDistance=1536
 soundClassObject=&Fort_Gadgets
}
